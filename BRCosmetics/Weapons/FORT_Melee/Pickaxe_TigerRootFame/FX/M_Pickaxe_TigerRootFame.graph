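# Translucent unlit Niagara mesh-particle material for the TigerRootFame pickaxe effect, built around a scrolling spike mask.
# Niagara dynamic parameters drive UV offset and scale, plus world position offset combining clamped bend, radial scaling and sine wobble.
# The noise-disturbable mask blends edge color with a curve-atlas gradient, adds distorted smoke highlights and shapes opacity.
# Particle color and time-of-day values modulate both outputs, with depth fade on default and deferred paths.
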Material /BRCosmetics/Weapons/FORT_Melee/Pickaxe_TigerRootFame/FX/M_Pickaxe_TigerRootFame
graph {
 properties {
  bUsedWithNiagaraMeshParticles=true
  blendMode=BLEND_Translucent
  pixelDepthOffsetMode=PDOM_Legacy
  shadingModel=MSM_Unlit
  usedShadingModels=Unlit
 }
 nodes [
  n1 = ComponentMask(None=VectorParameter() {parameterName=Object Position Offset}.RGB) {b=true, g=true, r=true}
  n2 = Subtract(
   A=WorldPosition() {worldPositionShaderOffset=WPT_ExcludeAllShaderOffsets}.XYZ
   B=Add(A=ParticlePositionWS(), B=Transform(None=@n1) {transformSourceType=TRANSFORMSOURCE_Instance})
  )
  n3 = Power(
   Base=ComponentMask(None=TextureCoordinate()) {r=true}
   Exp=DynamicParameter() {defaultValue=<rgba 1 50 3 1>, paramNames [UV_Offset, WPO Max, WPO Bend Power, Param4]}.WPO Bend Power
  )
  n4 = DynamicParameter() {parameterIndex=2, defaultValue=<rgba 1 1 0 1>, paramNames [Mask_U_Mult, Mask_V_Mult, Mesh Radial Scale, Mesh Radial Falloff Exponent]}
  n5 = Multiply(A=Multiply(A=Normalize(VectorInput=@n2), B=@n3), B=@n4.Mesh Radial Scale)
  n6 = TextureCoordinate()
  n7 = Multiply(A=Subtract(A=ComponentMask(None=@n6) {g=true}) {constB=0.5}) {constB=2}
  n8 = Power(Base=OneMinus(None=Abs(None=@n7)), Exp=@n4.Mesh Radial Falloff Exponent)
  n9 = DynamicParameter() {desc=Parameters set by Niagara, parameterIndex=1, paramNames [X,Y,Z,WPOIntensity]}
  n10 = DynamicParameter() {defaultValue=<rgba 1 50 3 1>, paramNames [UV_Offset, WPO Max, WPO Bend Power, Param4]}
  n11 = Clamp(
   None=MaterialFunctionCall(X=@n9.X, Y=@n9.X, Z=@n9.X) {materialFunction=&MakeFloat3}.Result
   Min=Multiply(A=@n10.WPO Max) {constA=-1, constB=-1}
   Max=@n10.WPO Max
  ) {maxDefault=50, minDefault=-50}
  n12 = Multiply(A=@n3, B=@n9.WPOIntensity)
  n13 = Constant3Vector() {constant=<rgba 1 0 0 0>}
  n14 = ComponentMask(None=TextureCoordinate() {uTiling=0, vTiling=4}) {g=true}
  n15 = DynamicParameter() {defaultValue=<rgba 1 50 3 1>, paramNames [UV_Offset, WPO Max, WPO Bend Power, Param4]}
  n16 = Sine(None=Add(A=Add(A=@n14, B=@n15.UV_Offset), B=Time()))
  n17 = TextureCoordinate()
  n18 = ComponentMask(None=@n17) {r=true}
  n19 = OneMinus(None=ComponentMask(None=@n17) {g=true})
  n20 = Transform(None=Multiply(A=Multiply(A=@n13, B=@n16), B=Multiply(A=@n18, B=@n19))) {transformSourceType=TRANSFORMSOURCE_Local}
  Add(
   A=LinearInterpolate(B=Multiply(A=@n5, B=@n8), Alpha=ScalarParameter() {parameterName=Toggle Mesh Radial Scale})
   B=Add(A=LinearInterpolate(B=@n11, Alpha=@n12), B=Multiply(A=@n20))
  )
  Constant2Vector() {g=0.5, r=1.5}
  n21 = ParticleColor()
  n22 = VectorParameter() {parameterName=EdgeColor, defaultValue=<rgba 1 0.3501 0 1>}
  n23 = TextureCoordinate()
  n24 = DynamicParameter() {defaultValue=<rgba 1 50 3 1>, paramNames [UV_Offset, WPO Max, WPO Bend Power, Param4]}
  n25 = TextureSampleParameter2D(UVs=Panner(Coordinate=Add(A=@n23, B=@n24.UV_Offset)) {speedX=-0.5}) {group=Emissive, parameterName=Tex, texture=/Game/CharClasses/Outlander/Gadgets/AntiMaterialCharge/Materials/T_AntiMaterialCharge_Dissolve_01}
  n26 = LinearInterpolate(A=ScalarParameter() {defaultValue=0.7, group=Emissive, parameterName=Emissive_Dark}, Alpha=@n25.R) {constA=0.3}
  n27 = MaterialFunctionCall(UVs=TextureCoordinate(), Rotation Center=Constant2Vector() {g=0.5, r=0.5}, "Rotation Angle (0-1)"=Constant()) {materialFunction=&CustomRotator}
  n28 = CurveAtlasRowParameter(CurveTime=Power(
    Base=ComponentMask(None=@n27.Rotated Values) {r=true}
    Exp=ScalarParameter() {defaultValue=2, group=Emissive, parameterName=Linear_gradient_power}
   ) {constExponent=3}) {
   atlas=&/CA_Pickaxe_Curves
   curve=&/C_Pickaxe_TigerRootFrame_Pink
   group=Emissive
   parameterName=Linear_gradient_Color
  }
  n29 = Multiply(A=@n28)
  n30 = VectorParameter() {group=Mask, parameterName=Mask_Speed, defaultValue=<rgba 0 0.5 0 0>}
  n31 = TextureCoordinate()
  n32 = Constant2Vector() {g=0.5}
  n33 = VectorParameter() {group=Mask, parameterName=Mask_UV_Scale, defaultValue=<rgba 1 1 1 1>}
  n34 = DynamicParameter() {parameterIndex=2, defaultValue=<rgba 1 1 0 1>, paramNames [Mask_U_Mult, Mask_V_Mult, Mesh Radial Scale, Mesh Radial Falloff Exponent]}
  n35 = Multiply(A=@n33.R, B=@n34.Mask_U_Mult)
  n36 = Multiply(A=@n33.G, B=@n34.Mask_V_Mult)
  n37 = Add(A=Multiply(A=Subtract(A=@n31, B=@n32) {desc=Offset uv.v Center}, B=AppendVector(A=@n35, B=@n36)), B=@n32)
  n38 = AppendVector(A=Constant(), B=DynamicParameter() {defaultValue=<rgba 1 50 3 1>, paramNames [UV_Offset, WPO Max, WPO Bend Power, Param4]}.UV_Offset)
  n39 = MaterialFunctionCall(Float2=Panner(Coordinate=Add(A=@n37, B=@n38), Speed=ComponentMask(None=@n30.RGB) {g=true, r=true})) {materialFunction=&BreakOutFloat2Components}
  n40 = Subtract(A=MaterialFunctionCall(Float2=TextureCoordinate()) {materialFunction=&BreakOutFloat2Components}.G) {constB=0.5}
  n41 = ScalarParameter() {defaultValue=-0.6, group=Mask, parameterName=Edge_Mask_UV_scale}
  n42 = Add(A=Multiply(A=Abs(None=@n40)) {constB=2}, B=@n41) {constB=-0.5}
  n43 = Add(B=@n41) {constA=1}
  n44 = Power(Base=Multiply(A=Saturate(None=@n42), B=Divide(B=@n43) {constB=1}) {constB=2}) {constExponent=4}
  n45 = Multiply(A=@n39.R, B=Add(A=Multiply(A=@n44) {constB=50}))
  n46 = TextureCoordinate()
  n47 = Multiply(
   A=ComponentMask(None=VectorParameter() {group=Mask, parameterName=Mask_Disturb_UV_Scale, defaultValue=<rgba 1.5 0.5 0 0>}.RGB) {g=true, r=true}
   B=MaterialFunctionCall(UVs=Add(A=@n46) {constB=0}, "Rotation Angle (0-1)"=Constant() {r=-0.25}) {materialFunction=&CustomRotator}.Rotated Values
  )
  n48 = DynamicParameter() {defaultValue=<rgba 1 50 3 1>, paramNames [UV_Offset, WPO Max, WPO Bend Power, Param4]}
  n49 = TextureSampleParameter2D(UVs=Panner(Coordinate=Add(A=@n47, B=@n48.UV_Offset)) {speedY=0.1}) {
   group=Mask
   parameterName=Mask_Noise
   texture=&T_AntiMaterialCharge_Dissolve_01
  }
  n50 = ScalarParameter() {group=Mask, parameterName=Mask_Disturb}
  n51 = TextureSampleParameter2D(UVs=Add(
    A=MaterialFunctionCall(X=@n45, Y=@n39.G) {materialFunction=&MakeFloat2}.Result
    B=Multiply(A=@n49.R, B=@n50) {constB=0}
   ) {constB=0}) {group=Mask, parameterName=mask, texture=&/T_TigerRootFame_Spikes_Mask}
  n52 = Constant()
  n53 = Multiply(A=TextureCoordinate())
  n54 = Multiply(
   A=ComponentMask(None=VectorParameter() {parameterName=HighLight_UV_Scale, defaultValue=<rgba 1 0.6 0 0>}.RGB) {g=true, r=true}
   B=MaterialFunctionCall(UVs=Add(A=@n53) {constB=0}, "Rotation Angle (0-1)"=Constant() {r=-0.25}) {materialFunction=&CustomRotator}.Rotated Values
  )
  n55 = DynamicParameter() {defaultValue=<rgba 1 50 3 1>, paramNames [UV_Offset, WPO Max, WPO Bend Power, Param4]}
  n56 = TextureCoordinate()
  n57 = Multiply(
   A=Constant2Vector() {g=0.6, r=1}
   B=MaterialFunctionCall(UVs=Add(A=@n56) {constB=0}, "Rotation Angle (0-1)"=Constant() {r=-0.25}) {materialFunction=&CustomRotator}.Rotated Values
  )
  n58 = TextureSample(UVs=Panner(Coordinate=Add(A=@n55.UV_Offset, B=@n57)) {speedY=0.1}) {desc=HighLight Distort Texture, texture=&T_AntiMaterialCharge_Dissolve_01}
  n59 = ScalarParameter() {defaultValue=0.3, parameterName=Emissive_Disturb}
  n60 = Panner(Coordinate=Add(A=Add(A=@n54, B=@n55.UV_Offset), B=Multiply(A=@n58.R, B=@n59) {constB=0.3})) {speedY=0.3}
  n61 = Multiply(
   A=VectorParameter() {parameterName=Hilight_Color, defaultValue=<rgba 0 2.1555 8 1>}.RGB
   B=Power(
    Base=TextureSampleParameter2D(UVs=@n60) {group=Emissive, parameterName=HIghLight_Tex, texture=/Game/Effects/Fort_Effects/Textures/Tile/T_Inky_Smoke_Tile2}.R
    Exp=ScalarParameter() {defaultValue=2, group=Emissive, parameterName=HighLight_Power}
   )
  )
  n62 = ScalarParameter() {defaultValue=1, group=Emissive, parameterName=Added_the_HighLight_Tex}
  n63 = Add(
   A=LinearInterpolate(A=Multiply(A=@n22.RGB, B=@n26), B=Multiply(A=@n29, B=@n26), Alpha=@n51.G) {constAlpha=1}
   B=Multiply(A=LinearInterpolate(A=@n52, B=@n61, Alpha=@n62), B=@n51.R)
  )
  n64 = VectorParameter() {group=Time of day, parameterName=Morning_Color, defaultValue=<rgba 1 1 1 1>}
  n65 = VectorParameter() {group=Time of day, parameterName=Day_Color, defaultValue=<rgba 1 1 1 1>}
  n66 = VectorParameter() {group=Time of day, parameterName=Evening_Color, defaultValue=<rgba 1 1 1 1>}
  n67 = VectorParameter() {group=Time of day, parameterName=Night_Color, defaultValue=<rgba 1 1 1 1>}
  MaterialFunctionCall(Emissive=Multiply(
    A=Multiply(A=@n21.RGB, B=@n63)
    B=MaterialFunctionCall(Color_Morning=@n64.RGB, Color_Day=@n65.RGB, Color_Evening=@n66.RGB, Color_Night=@n67.RGB) {materialFunction=&Color_Time_of_Day}.Result
   )) {materialFunction=&MF_FX_EmissiveAdjustments}
  n68 = TextureCoordinate()
  n69 = Power(Base=Sine(None=ComponentMask(None=@n68) {g=true}) {period=2}) {constExponent=4}
  n70 = ComponentMask(None=TextureCoordinate()) {g=true}
  n71 = Saturate(None=Multiply(A=Power(Base=@n70) {constExponent=2.5}) {constB=10})
  n72 = Multiply(A=Saturate(None=Multiply(A=@n69, B=@n71)), B=@n51.R) {constA=1}
  n73 = ScalarParameter() {defaultValue=1, group=Mask, parameterName=Mask Brighten, sliderMax=10, sliderMin=1, sortPriority=3}
  n74 = Multiply(A=@n21.A, B=Saturate(None=Multiply(A=@n72, B=@n73)))
  n75 = DepthFade(Opacity=@n74) {fadeDistanceDefault=5}
  Multiply(
   A=MaterialFunctionCall(Value_Morning=@n64.A, Value_Day=@n65.A, Value_Evening=@n66.A, Value_Night=@n67.A) {materialFunction=&Scalar_Time_of_Day}.Result
   B=ShadingPathSwitch(Default=@n75, Deferred=@n75, Forward=@n74, Mobile=@n74)
  )
 ]
}
