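# Niagara effect for the TigerRootFame pickaxe cosmetic, combining a holographic mesh with a swaying, music-reactive flame.
# Core follows the skeletal mesh, sharing system-set scale, rotation and random values with preset aberration/ghost, rainbow-line, glow and time-of-day-scaled light layers.
# FlamePendulum runs a spring-driven pendulum from the sampled mesh position, passing its offset from rest into material parameters.
# Music spectrum, gated by User.isReactive, drives further flame material parameters.
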NiagaraSystem /BRCosmetics/Weapons/FORT_Melee/Pickaxe_TigerRootFame/FX/NS_Pickaxe_TigerRootFame
dependencies {
 usedDataInterfaces [&NiagaraDataInterfaceCamera, &NiagaraDataInterfaceCurve]
 usedDynamicInputs [
  &RandomRangeFloat
  &MultiplyQuaternion
  &MakeQuatFromAxisAngle
  &Lerp_Vector
  &Sine
  &Add_Float
 ]
 usedModules [
  &self:SystemSpawnScript.SystemScriptSource.SystemScriptGraph.NiagaraNodeAssignment_1.SetVariables_#1
  &self:SystemSpawnScript.SystemScriptSource.SystemScriptGraph.NiagaraNodeAssignment_6.SetVariables_#2
  &self:SystemSpawnScript.SystemScriptSource.SystemScriptGraph.NiagaraNodeAssignment_7.SetVariables_#3
  &SystemState
  &Timeline
  &self:SystemSpawnScript.SystemScriptSource.SystemScriptGraph.NiagaraNodeAssignment_4.SetVariables_#4
  &self:SystemSpawnScript.SystemScriptSource.SystemScriptGraph.NiagaraNodeAssignment_0.SetVariables_#5
 ]
}
emitters {
 Core {
  emitterUpdateScript [
   &EmitterState(
    Inactive Response=::NewEnumerator1
    Life Cycle Mode=::NewEnumerator1
    Loop Behavior=::NewEnumerator1
    Loop Duration Mode=::NewEnumerator1
    UseLoopDelay=false
   )
   &SpawnBurst_Instantaneous(Use Spawn Probability=false, UseLoopCountLimit=false)
  ]
  particleSpawnScript [
   &InitializeParticle(
    Color Mode=::NewEnumerator0
    Lifetime=5
    Mesh Scale Mode=::NewEnumerator3
    Mesh Uniform Scale=$System.InitMeshScale
    Position Offset=&AddVectorToPosition(Position=$System.InitPosition, Vector=<0 0 0>)
    UsePositionOffset=true
    Write Material Random=true
   )
   &DynamicMaterialParameters(
    Index 0 Param 1=$System.Random
    Index 0 Param 2=&MakeFloatFromVector(Vector=$Particles.Scale, Channel=::NewEnumerator0)
    Module.Param0WriteEnabled=true
    Module.Param1WriteEnabled=true
    Module.Param2WriteEnabled=false
    Module.Param3WriteEnabled=false
   )
  ]
  particleUpdateScript [
   &ParticleState(Kill Particles When Lifetime Has Elapsed=false, Let Infinitely Lived Particles Die When Emitter Deactivates=true)
   &SkeletalMeshLocation(Bone Sampling Mode=::NewEnumerator4, MaskSpawnGroup=false, Position Sampling=::NewEnumerator2, Skeletal Mesh=$User.SkeletalMesh)
   /SetVariables(Particles.MeshOrientation=$System.AutoRotate) {isSetParameters=true}
   &self:LODSetting(LOD1=400, LOD2=1000) {module=LODSetting}
  ]
  rendererClasses [&NiagaraMeshRendererProperties]
  renderers [{rendererClass=&NiagaraMeshRendererProperties, rendererIndex=0}]
 }
 FlamePendulum {
  emitterUpdateScript [
   &EmitterState(
    Inactive Response=::NewEnumerator1
    Life Cycle Mode=::NewEnumerator1
    Loop Behavior=::NewEnumerator1
    Loop Duration Mode=::NewEnumerator1
    UseLoopDelay=false
   )
   &SpawnBurst_Instantaneous(Use Spawn Probability=false, UseLoopCountLimit=true)
  ]
  particleSpawnScript [
   &InitializeParticle(
    Color Mode=::NewEnumerator0
    Lifetime=5
    Sprite Size Mode=::NewEnumerator3
    Uniform Sprite Size=5
    UsePositionOffset=false
    Write Material Random=true
   )
   &InitialMeshOrientation(Mesh Orientation Mode=::NewEnumerator7, Rotation=<0 0 -0.25>, Use Rotation Vector=true)
   /SetVariables(Particles.PendulumLength=25, Particles.PendulumVector:Vector3f=<0 0 1>, Particles.SkelLocationOffset:Vector3f=<0 0 75>) {isSetParameters=true}
   /SetVariables(Particles.PendulumRest:Vector3f=&Add_Vector(
     A=&Multiply_VectorByFloat(Vector=$Particles.PendulumVector, Float=$Particles.PendulumLength)
     B=$StackContext.SkelLocationOffset
    )) {isSetParameters=true}
  ]
  particleUpdateScript [
   &ParticleState(Kill Particles When Lifetime Has Elapsed=false, Let Infinitely Lived Particles Die When Emitter Deactivates=true)
   &SkeletalMeshLocation(
    Bone Sampling Mode=::NewEnumerator4
    MaskSpawnGroup=false
    Position Sampling=::NewEnumerator2
    Sampled Position Offset=$Particles.SkelLocationOffset
    Skeletal Mesh=$User.SkeletalMesh
   )
   /SetVariables(Particles.Position_Mesh:NiagaraPosition=$Output.SkeletalMeshLocation.SampledPositionWithoutOffset) {isSetParameters=true}
   &PendulumSetup(
    Module.Enable Spring Driver=true
    Pendulum Length:NiagaraFloat=$Particles.PendulumLength
    Pendulum Pivot:NiagaraPosition=$Particles.SkeletalMeshLocation.SampledPosition
    Pendulum Pivot Offset Coordinate Space:ENiagaraCoordinateSpace=::Local
    Pendulum Rest Axis:Vector3f=&TransformVector(Vector=$Particles.PendulumVector, Destination Space=::World)
    Rigid Pendulum=false
    "Spring Driven Constraint (Tightness)"=20
   )
   &GravityForce() {enabled=false}
   &SolveForcesAndVelocity(Clamp Velocity=false, Limit Acceleration=false)
   &PendulumConstraint()
   &CalculateAccurateVelocity()
   /SetVariables(Particles.DynamicMaterialParameter1:Vector4f=&MakeFloat4FromVectorAndFloat(
     "Vector (RGB)"=&SubtractVectorFromPosition(
      Position=$Particles.Position
      Vector=&TransformPosition(Position=Linked() {linkedVariable {name=Particles.PendulumRest, type=Vector3f}}, Source Space=::Local, Destination Space=::Simulation)
     )
     "Float (Alpha)"=&Lerp_Float(A=1, B=3, Alpha=$System.Timeline.NormalizedTime)
    )) {isSetParameters=true}
   &DynamicMaterialParameters(
    Index 2 Param 1=&FN_FrontEndFloatModifier(
     Gameplay=&Lerp_Float(
      A=1.25
      B=1
      Alpha=&MakeFloatFromLinearColor(
       LinearColor=&MultiplyLinearColor(
        Linear Color=$NPC.NPC_Audio_Reactive_Parameters.Spectrum_Music
        Scale Factor=&MakeVector4_FromFloat(Float=$User.isReactive)
       )
       Channel:ENiagaraLinearColor_Channels=::NewEnumerator3
      )
     )
     Front End=1.25
    )
    Index 2 Param 2=&Lerp_Float(
     A=0.7
     B=2.25
     Alpha=&MakeFloatFromLinearColor(
      LinearColor=&MultiplyLinearColor(
       Linear Color=$NPC.NPC_Audio_Reactive_Parameters.Spectrum_Music
       Scale Factor=&MakeVector4_FromFloat(Float=$User.isReactive)
      )
      Channel:ENiagaraLinearColor_Channels=::NewEnumerator3
     )
    )
    Index 2 Param 3=&Lerp_Float(
     A=0
     B=100
     Alpha=&MakeFloatFromLinearColor(
      LinearColor=&MultiplyLinearColor(
       Linear Color=$NPC.NPC_Audio_Reactive_Parameters.Spectrum_Music
       Scale Factor=&MakeVector4_FromFloat(Float=$User.isReactive)
      )
      Channel:ENiagaraLinearColor_Channels=::NewEnumerator3
     )
    )
    Index 2 Param 4=5
    Module.Param0WriteEnabled2=true
    Module.Param1WriteEnabled2=true
    Module.Param2WriteEnabled2=true
    Module.Param3WriteEnabled2=true
    Write Parameter Index 0=false
    Write Parameter Index 2=true
   )
  ]
  rendererClasses [&NiagaraMeshRendererProperties, &NiagaraSpriteRendererProperties]
  renderers [
   {rendererClass=&NiagaraMeshRendererProperties, rendererIndex=0}
   {rendererClass=&NiagaraMeshRendererProperties, rendererIndex=1}
   {rendererClass=&NiagaraMeshRendererProperties, rendererIndex=2}
   {rendererClass=&NiagaraSpriteRendererProperties, rendererIndex=3}
   {rendererClass=&NiagaraMeshRendererProperties, rendererIndex=4}
  ]
 }
 PresetFX_AberrationAndGhost {
  emitterUpdateScript [
   &EmitterState(
    Inactive Response=::NewEnumerator1
    Life Cycle Mode=::NewEnumerator1
    Loop Behavior=::NewEnumerator1
    Loop Duration Mode=::NewEnumerator1
    UseLoopDelay=false
   )
   &SpawnBurst_Instantaneous(Use Spawn Probability=false, UseLoopCountLimit=false)
  ]
  particleSpawnScript [
   &InitializeParticle(
    Color Mode=::NewEnumerator0
    Lifetime=5
    Mesh Scale Mode=::NewEnumerator3
    Mesh Uniform Scale=&Multiply_Float(A=$System.InitMeshScale, B=2)
    Position Offset=&AddVectorToPosition(Position=$System.InitPosition, Vector=<0.7 0 0>)
    Sprite Size Mode=::NewEnumerator3
    Uniform Sprite Size=70
    UsePositionOffset=true
    Write Material Random=true
   )
   &DynamicMaterialParameters(
    Index 0 Param 1=$System.Random
    Module.Param0WriteEnabled=true
    Module.Param1WriteEnabled=false
    Module.Param2WriteEnabled=false
    Module.Param3WriteEnabled=false
   )
  ]
  particleUpdateScript [
   &ParticleState(Kill Particles When Lifetime Has Elapsed=false, Let Infinitely Lived Particles Die When Emitter Deactivates=true)
   /SetVariables(Particles.MeshOrientation=$System.AutoRotate) {isSetParameters=true}
  ]
  rendererClasses [&NiagaraMeshRendererProperties]
  renderers [
   {rendererClass=&NiagaraMeshRendererProperties, rendererIndex=0}
   {rendererClass=&NiagaraMeshRendererProperties, rendererIndex=1}
  ]
 }
 PresetFX_Glow {
  emitterUpdateScript [
   &EmitterState(
    Inactive Response=::NewEnumerator1
    Life Cycle Mode=::NewEnumerator1
    Loop Behavior=::NewEnumerator1
    Loop Duration Mode=::NewEnumerator1
    UseLoopDelay=false
   )
   &SpawnBurst_Instantaneous(Use Spawn Probability=false, UseLoopCountLimit=false)
  ]
  particleSpawnScript [&InitializeParticle(
    Color=<rgba 1 0.7333 0 0.15>
    Lifetime=5
    Position Offset=&AddVectorToPosition(Position=$System.InitPosition, Vector=<0.7 0 0>)
    Sprite Size Mode=::NewEnumerator3
    Uniform Sprite Size=75
    UsePositionOffset=true
    Write Material Random=true
   )]
  particleUpdateScript [&ParticleState(Kill Particles When Lifetime Has Elapsed=false, Let Infinitely Lived Particles Die When Emitter Deactivates=true)]
  rendererClasses [&NiagaraSpriteRendererProperties]
  renderers [{rendererClass=&NiagaraSpriteRendererProperties}]
 }
 PresetFX_Light {
  emitterUpdateScript [
   &EmitterState(Life Cycle Mode=::NewEnumerator1, Loop Behavior=::NewEnumerator1, Loop Duration Mode=::NewEnumerator1, UseLoopDelay=false)
   &SpawnBurst_Instantaneous(Use Spawn Probability=false, UseLoopCountLimit=false)
  ]
  particleSpawnScript [
   &InitializeParticle(
    Color=<rgba 14 10.2667 0 0.25>
    Position Offset=&AddVectorToPosition(Position=Linked() {linkedVariable {name=System.InitPosition, type=Vector3f}}, Vector=<0.7 0 0>)
    UsePositionOffset=true
    Write Material Random=true
   )
   /SetVariables(Particles.LightRadius=35) {isSetParameters=true}
   /SetVariables(Particles.LightPosition:Vector3f=&Add_Vector(A=Linked() {linkedVariable {name=Particles.Position, type=NiagaraPosition}}, B=<0 -20 0>)) {isSetParameters=true}
  ]
  particleUpdateScript [
   &ParticleState(Kill Particles When Lifetime Has Elapsed=false, Let Infinitely Lived Particles Die When Emitter Deactivates=true)
   &ScaleColor(
    Scale RGB=&Multiply_Vector(
     A=&VectorFromFloat(Value=&FNTimeOfDayScalar(Night Scale=0.7, Morning Scale=0.3, Evening Scale=0.1))
     B=<1 1 1>
    )
    ScaleA=true
    ScaleRGB=true
   )
  ]
  rendererClasses [&NiagaraLightRendererProperties]
  renderers [{rendererClass=&NiagaraLightRendererProperties, rendererIndex=0}]
 }
 PresetFX_RainbowLine {
  emitterUpdateScript [
   &EmitterState(
    Inactive Response=::NewEnumerator1
    Life Cycle Mode=::NewEnumerator1
    Loop Behavior=::NewEnumerator1
    Loop Duration Mode=::NewEnumerator1
    UseLoopDelay=false
   )
   &SpawnBurst_Instantaneous(Use Spawn Probability=false, UseLoopCountLimit=false)
   /SetVariables(Emitter.RainbowLine_Flat:NiagaraFloat=&MakeCustomFloatFromBool(Bool=&SetBoolByFloatComparison(A=$User.IsFrom3D, B=0), True Float=0, False Float=-1)) {isSetParameters=true}
   /SetVariables(Emitter.RainbowLine_Strength:NiagaraFloat=&MakeCustomFloatFromBool(Bool=&SetBoolByFloatComparison(A=$User.IsFrom3D, B=0), True Float=4, False Float=2)) {isSetParameters=true}
   /SetVariables(Emitter.RainbowLine_Power=0.1) {isSetParameters=true}
   /SetVariables(Emitter.RainbowLine_Density=5) {isSetParameters=true}
   /SetVariables(Emitter.RainbowLine_FlowSpeed=0.15) {isSetParameters=true}
  ]
  particleSpawnScript [
   &InitializeParticle(
    Color Mode=::NewEnumerator0
    Lifetime=5
    Mesh Scale Mode=::NewEnumerator3
    Mesh Uniform Scale=$System.InitMeshScale
    Position Offset=&AddVectorToPosition(Position=$System.InitPosition, Vector=<0.7 0 0>)
    Sprite Size Mode=::NewEnumerator3
    Uniform Sprite Size=70
    UsePositionOffset=true
    Write Material Random=true
   )
   &DynamicMaterialParameters(
    Index 0 Param 1=$System.Random
    Module.Param0WriteEnabled=true
    Module.Param1WriteEnabled=false
    Module.Param2WriteEnabled=false
    Module.Param3WriteEnabled=false
   )
   /SetVariables(Particles.RainbowLineSize:Vector2f=&Multiply_Vector2D_ByFloat(Float=0.5, Vector2D=$Particles.SpriteSize)) {isSetParameters=true}
  ]
  particleUpdateScript [
   &ParticleState(Kill Particles When Lifetime Has Elapsed=false, Let Infinitely Lived Particles Die When Emitter Deactivates=true)
   /SetVariables(Particles.MeshOrientation=$System.AutoRotate) {isSetParameters=true}
  ]
  rendererClasses [&NiagaraMeshRendererProperties]
  renderers [{rendererClass=&NiagaraMeshRendererProperties, rendererIndex=0}]
 }
}
summary {
 systemName=NS_Pickaxe_TigerRootFame
 emitters [
  {emitterName=FlamePendulum, rendererClasses [&NiagaraMeshRendererProperties, &NiagaraSpriteRendererProperties]}
  {emitterName=Core, rendererClasses [&NiagaraMeshRendererProperties]}
  {emitterName=PresetFX_AberrationAndGhost, rendererClasses [&NiagaraMeshRendererProperties]}
  {emitterName=PresetFX_RainbowLine, rendererClasses [&NiagaraMeshRendererProperties]}
  {emitterName=PresetFX_Glow, rendererClasses [&NiagaraSpriteRendererProperties]}
  {emitterName=PresetFX_Light, rendererClasses [&NiagaraLightRendererProperties]}
 ]
 userVariables [
  User.Backpack_DiffuseTex {
   type=Texture
   description=""
   defaultValue {
    struct=NiagaraExt_VariableValue_Object
    value {object=/BRCosmetics/Weapons/FORT_Melee/Pickaxe_TigerRootFame/FX/Texture/T_Pickaxe_TigerRootFame_Holo_D, objectClass=/Script/Engine.Texture}
   }
  }
  User.Backpack_Mat {
   type=MaterialInterface
   description=""
   defaultValue {
    struct=NiagaraExt_VariableValue_Object
    value {
     object=/BRCosmetics/Weapons/FORT_Melee/Pickaxe_TigerRootFame/FX/Material/MI_Pickaxe_TigerRootFame_Holo
     objectClass=/Script/Engine.MaterialInterface
    }
   }
  }
  User.Backpack_NormalTex {
   type=Texture
   description=""
   defaultValue {
    struct=NiagaraExt_VariableValue_Object
    value {
     object=/BRCosmetics/Weapons/FORT_Melee/Pickaxe_TigerRootFame/FX/Texture/T_Pickaxe_TigerRootFame_Holo_Normal
     objectClass=&Texture
    }
   }
  }
  User.Bakcpack_WPOTex {
   type=Texture
   description=""
   defaultValue {
    struct=NiagaraExt_VariableValue_Object
    value {object=/BRCosmetics/Weapons/FORT_Melee/Pickaxe_TigerRootFame/FX/Texture/T_Pickaxe_TigerRootFame_Holo_WPO, objectClass=&Texture}
   }
  }
  User.bIsFrontEnd {type=NiagaraBool}
  User.bIsFrontEndPreview {type=NiagaraBool}
  User.bisSwinging {type=NiagaraBool}
  User.IsFrom3D {type=NiagaraFloat, defaultValue=0, description=""}
  User.isReactive {type=NiagaraFloat, defaultValue=1, description=""}
  User.SkeletalMesh {
   type=NiagaraDataInterfaceSkeletalMesh
   defaultValue {
    struct=NiagaraExt_VariableValue_DataInterface
    value {dataInterface=&self:NiagaraDataInterfaceSkeletalMesh_0}
   }
  }
 ]
}
